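# Physics asset for the RoseDepth Seed female medium body's dynamic skirt, with capsule bodies for cloth bones and legs.
# Five skirt chains (back, back-left, front-left, left, right) hang from base bodies jointed to root, using tight swing limits and zero twist.
# Free-rotating thigh, calf and foot capsules provide leg volumes, and a huge root capsule acts as a floor.
# A grenade constraint references bones without bodies here, apparently left over.
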PhysicsAsset /BRCosmetics/Characters/Player/Female/Medium/Bodies/F_MED_RoseDepth_Seed/Meshes/F_MED_RoseDepth_Seed_Dyn_skirt
bodies {
 base_skirt_bk_01 [{shapeType=Capsule, length=2, radius=2, shapeName=base_skirt_bk_01_capsule, rotation=<rot 0 0 90>}]
 base_skirt_bk_l_01 [{shapeType=Capsule, length=0.505, radius=0.505, shapeName=base_skirt_bk_l_01_capsule, rotation=<rot 0 -90 0>}]
 base_skirt_fr_l_01 [{shapeType=Capsule, length=0.505, radius=0.505, shapeName=base_skirt_fr_l_01_capsule, rotation=<rot 0 -90 0>}]
 base_skirt_l_01 [{shapeType=Capsule, length=0.505, radius=0.505, shapeName=base_skirt_l_01_capsule, rotation=<rot 0 -90 0>}]
 base_skirt_r_02 [{shapeType=Capsule, length=0.505, radius=0.505, shapeName=base_skirt_r_02_capsule, rotation=<rot 0 -90 0>}]
 calf_l [{shapeType=Capsule, length=36, radius=10, center=<-24 -2 0>, rotation=<rot 90 0 0>}]
 calf_r [{shapeType=Capsule, length=36, radius=10, center=<24 2 0>, rotation=<rot 90 0 0>}]
 dyn_skirt_bk_01 [{shapeType=Capsule, length=24, radius=12, shapeName=dyn_skirt_bk_01_capsule, center=<8 0 9.1225>, rotation=<rot -0.0002 0 75>}]
 dyn_skirt_bk_02 [{shapeType=Capsule, length=24, radius=12, shapeName=dyn_skirt_bk_02_capsule, center=<8 0 9.1225>, rotation=<rot -0.0003 0.0001 75>}]
 dyn_skirt_bk_03 [{shapeType=Capsule, length=24, radius=12, shapeName=dyn_cape_b_l_03_capsule, center=<8 0 9.1225>, rotation=<rot -0.0001 0 75>}]
 dyn_skirt_bk_04 [{shapeType=Capsule, length=24, radius=12, shapeName=dyn_cape_b_l_03_capsule, center=<8 0 9.1225>, rotation=<rot -0.0001 0 75>}]
 dyn_skirt_bk_05 [{shapeType=Capsule, length=24, radius=12, shapeName=dyn_cape_b_l_03_capsule, center=<8 0 9.1225>, rotation=<rot -0.0001 0 75>}]
 dyn_skirt_bk_l_01 [{shapeType=Capsule, length=1, radius=1, shapeName=dyn_skirt_bk_l_01_capsule, rotation=<rot 90 0 0>}]
 dyn_skirt_bk_l_02 [{shapeType=Capsule, length=1, radius=1, shapeName=dyn_skirt_bk_l_02_capsule, rotation=<rot 90 0 0>}]
 dyn_skirt_bk_l_03 [{shapeType=Capsule, length=24, radius=12, shapeName=dyn_skirt_bk_l_03_capsule, center=<8 0 9.0483>, rotation=<rot 0 0 90>}]
 dyn_skirt_bk_l_04 [{shapeType=Capsule, length=24, radius=12, shapeName=dyn_cape_b_l_03_capsule, center=<8 0 9.0483>, rotation=<rot 0 0 90>}]
 dyn_skirt_bk_l_05 [{shapeType=Capsule, length=24, radius=12, shapeName=dyn_skirt_bk_r_05_capsule, center=<8 0 9.0483>, rotation=<rot 0 0 90>}]
 dyn_skirt_fr_l_01 [{shapeType=Capsule, length=1, radius=1, shapeName=dyn_skirt_fr_l_01_capsule, rotation=<rot 90 0 0>}]
 dyn_skirt_fr_l_02 [{shapeType=Capsule, length=1, radius=1, shapeName=dyn_skirt_fr_l_02_capsule, rotation=<rot 90 0 0>}]
 dyn_skirt_fr_l_03 [{shapeType=Capsule, length=7.6336, radius=4.0291, shapeName=dyn_skirt_fr_l_03_capsule, center=<4.3716 0 0.876>, rotation=<rot 0 0 90>}]
 dyn_skirt_fr_l_04 [{shapeType=Capsule, length=24, radius=12, shapeName=dyn_skirt_fr_l_04_capsule, center=<8 0 7.1782>, rotation=<rot 0 0 90>}]
 dyn_skirt_fr_l_05 [{shapeType=Capsule, length=24, radius=12, shapeName=dyn_skirt_fr_l_05_capsule, center=<8 0 7.1782>, rotation=<rot 0 0 90>}]
 dyn_skirt_l_01 [{shapeType=Capsule, length=1, radius=1, shapeName=dyn_skirt_l_01_capsule, rotation=<rot 90 0 0>}]
 dyn_skirt_l_02 [{shapeType=Capsule, length=1, radius=1, shapeName=dyn_skirt_l_02_capsule, rotation=<rot 90 0 0>}]
 dyn_skirt_l_03 [{shapeType=Capsule, length=7.6336, radius=4.0291, shapeName=dyn_skirt_l_03_capsule, center=<2 0 0>, rotation=<rot 0 0 90>}]
 dyn_skirt_l_04 [{shapeType=Capsule, length=24, radius=12, shapeName=dyn_skirt_l_04_capsule, center=<8 0 8.2184>, rotation=<rot 0 0 90>}]
 dyn_skirt_l_05 [{shapeType=Capsule, length=24, radius=12, shapeName=dyn_skirt_l_05_capsule, center=<8 0 8.2184>, rotation=<rot 0 0 90>}]
 dyn_skirt_r_01 [{shapeType=Capsule, length=24, radius=12, shapeName=dyn_skirt_r_01_capsule, center=<8 -6.0514 6.0514>, rotation=<rot -0.0001 0 50>}]
 dyn_skirt_r_02 [{shapeType=Capsule, length=24, radius=12, shapeName=dyn_skirt_r_02_capsule, center=<8 -6.0514 6.0514>, rotation=<rot -0.0001 0 45>}]
 dyn_skirt_r_03 [{shapeType=Capsule, length=24, radius=12, shapeName=dyn_skirt_r_03_capsule, center=<8 -6.0514 6.0514>, rotation=<rot -0.0001 0 45>}]
 dyn_skirt_r_04 [{shapeType=Capsule, length=24, radius=12, shapeName=dyn_skirt_r_04_capsule, center=<8 -5.5009 6.5557>, rotation=<rot -0.0001 0 50>}]
 dyn_skirt_r_05 [
  {
   shapeType=Capsule
   length=32
   radius=12
   shapeName=dyn_skirt_l_05_capsule
   center=<13.702 0.6003 14.4301>
   rotation=<rot 6.164 13.9537 66.6287>
  }
  {shapeType=Capsule, length=32, radius=12, shapeName=dyn_skirt_l_05_capsule, center=<8.7066 -4.5605 4.615>, rotation=<rot -0.0001 8.8072 45>}
 ]
 foot_l [{shapeType=Capsule, length=15, radius=10, shapeName=foot_l_capsule, center=<-4.0948 5.9174 -0.1893>, rotation=<rot 0 0 90>}]
 foot_r [{shapeType=Capsule, length=15.6721, radius=10, shapeName=foot_r_capsule, center=<4.0948 -5.9174 0.1893>, rotation=<rot 0 0 -90>}]
 root [{shapeType=Capsule, length=1, radius=400, shapeName=floor_joint_capsule, center=<0 0 -395>, rotation=<rot 0 90 -90>}]
 thigh_l [
  {
   shapeType=Capsule
   length=21.2415
   radius=6.9872
   shapeName=thigh_l_capsule
   center=<-13.2994 -0.7814 0.9974>
   rotation=<rot 88.8078 -5.1517 -2.3623>
  }
  {
   shapeType=Capsule
   length=16.9974
   radius=6.4131
   shapeName=thigh_l_capsule
   center=<-30.0564 0.0472 0.2459>
   rotation=<rot 85.5064 -2.9714 -1.564>
  }
 ]
 thigh_r [
  {
   shapeType=Capsule
   length=21.2415
   radius=6.9872
   shapeName=thigh_l_capsule
   center=<13.2994 0.7814 -0.9974>
   rotation=<rot 88.8078 -5.1517 177.6377>
  }
  {
   shapeType=Capsule
   length=16.9974
   radius=6.4131
   shapeName=thigh_l_capsule
   center=<30.0564 -0.0472 -0.2459>
   rotation=<rot 85.5064 -2.9714 178.436>
  }
 ]
}
constraints [
 {bone1Name=dyn_grenade_02, bone2Name=dyn_grenade_02_MSTR, swing1LimitDegrees=30, swing2LimitDegrees=30, twistLimitDegrees=5}
 {
  bone1Name=calf_l
  bone2Name=thigh_l
  swing1LimitDegrees=45
  swing1Motion=Free
  swing2LimitDegrees=45
  swing2Motion=Free
  twistLimitDegrees=45
  twistMotion=Free
 }
 {
  bone1Name=calf_r
  bone2Name=thigh_r
  swing1LimitDegrees=45
  swing1Motion=Free
  swing2LimitDegrees=45
  swing2Motion=Free
  twistLimitDegrees=45
  twistMotion=Free
 }
 {bone1Name=dyn_skirt_bk_l_01, bone2Name=base_skirt_bk_l_01, swing1LimitDegrees=15, swing2LimitDegrees=15, twistLimitDegrees=0}
 {bone1Name=dyn_skirt_bk_l_02, bone2Name=dyn_skirt_bk_l_01, swing1LimitDegrees=45, swing2LimitDegrees=45, twistLimitDegrees=0}
 {bone1Name=dyn_skirt_bk_l_03, bone2Name=dyn_skirt_bk_l_02, swing1LimitDegrees=5, swing2LimitDegrees=25, twistLimitDegrees=0}
 {bone1Name=dyn_skirt_bk_l_04, bone2Name=dyn_skirt_bk_l_03, swing1LimitDegrees=5, swing2LimitDegrees=25, twistLimitDegrees=0}
 {bone1Name=dyn_skirt_bk_l_05, bone2Name=dyn_skirt_bk_l_04, swing1LimitDegrees=5, swing2LimitDegrees=25, twistLimitDegrees=0}
 {bone1Name=dyn_skirt_fr_l_01, bone2Name=base_skirt_fr_l_01, swing1LimitDegrees=15, swing2LimitDegrees=15, twistLimitDegrees=0}
 {bone1Name=dyn_skirt_fr_l_02, bone2Name=dyn_skirt_fr_l_01, swing1LimitDegrees=45, swing2LimitDegrees=45, twistLimitDegrees=0}
 {bone1Name=dyn_skirt_fr_l_03, bone2Name=dyn_skirt_fr_l_02, swing1LimitDegrees=0, swing2LimitDegrees=10, twistLimitDegrees=0}
 {bone1Name=dyn_skirt_fr_l_04, bone2Name=dyn_skirt_fr_l_03, swing1LimitDegrees=25, swing2LimitDegrees=24, twistLimitDegrees=0}
 {bone1Name=dyn_skirt_fr_l_05, bone2Name=dyn_skirt_fr_l_04, swing1LimitDegrees=25, swing2LimitDegrees=24, twistLimitDegrees=0}
 {bone1Name=dyn_skirt_l_01, bone2Name=base_skirt_l_01, swing1LimitDegrees=15, swing2LimitDegrees=15, twistLimitDegrees=0}
 {bone1Name=dyn_skirt_l_02, bone2Name=dyn_skirt_l_01, swing1LimitDegrees=45, swing2LimitDegrees=45, twistLimitDegrees=0}
 {bone1Name=dyn_skirt_l_03, bone2Name=dyn_skirt_l_02, swing1LimitDegrees=0, swing2LimitDegrees=20, twistLimitDegrees=0}
 {bone1Name=dyn_skirt_l_04, bone2Name=dyn_skirt_l_03, swing1LimitDegrees=25, swing2LimitDegrees=25, twistLimitDegrees=0}
 {bone1Name=dyn_skirt_l_05, bone2Name=dyn_skirt_l_04, swing1LimitDegrees=25, swing2LimitDegrees=25, twistLimitDegrees=0}
 {bone1Name=dyn_skirt_r_01, bone2Name=base_skirt_r_02, swing1LimitDegrees=15, swing2LimitDegrees=15, twistLimitDegrees=0}
 {bone1Name=dyn_skirt_r_02, bone2Name=dyn_skirt_r_01, swing1LimitDegrees=25, swing2LimitDegrees=25, twistLimitDegrees=0}
 {bone1Name=dyn_skirt_r_03, bone2Name=dyn_skirt_r_02, swing1LimitDegrees=25, swing2LimitDegrees=25, twistLimitDegrees=0}
 {bone1Name=dyn_skirt_r_04, bone2Name=dyn_skirt_r_03, swing1LimitDegrees=25, swing2LimitDegrees=25, twistLimitDegrees=0}
 {bone1Name=dyn_skirt_r_05, bone2Name=dyn_skirt_r_04, swing1LimitDegrees=25, swing2LimitDegrees=25, twistLimitDegrees=0}
 {
  bone1Name=foot_l
  bone2Name=calf_l
  swing1LimitDegrees=45
  swing1Motion=Free
  swing2LimitDegrees=45
  swing2Motion=Free
  twistLimitDegrees=45
  twistMotion=Free
 }
 {
  bone1Name=foot_r
  bone2Name=calf_r
  swing1LimitDegrees=45
  swing1Motion=Free
  swing2LimitDegrees=45
  swing2Motion=Free
  twistLimitDegrees=45
  twistMotion=Free
 }
 {
  bone1Name=thigh_l
  bone2Name=root
  swing1LimitDegrees=45
  swing1Motion=Free
  swing2LimitDegrees=45
  swing2Motion=Free
  twistLimitDegrees=45
  twistMotion=Free
 }
 {
  bone1Name=thigh_r
  bone2Name=root
  swing1LimitDegrees=45
  swing1Motion=Free
  swing2LimitDegrees=45
  swing2Motion=Free
  twistLimitDegrees=45
  twistMotion=Free
 }
 {
  bone1Name=base_skirt_fr_l_01
  bone2Name=root
  swing1LimitDegrees=45
  swing1Motion=Free
  swing2LimitDegrees=45
  swing2Motion=Free
  twistLimitDegrees=45
  twistMotion=Free
 }
 {
  bone1Name=base_skirt_l_01
  bone2Name=root
  swing1LimitDegrees=45
  swing1Motion=Free
  swing2LimitDegrees=45
  swing2Motion=Free
  twistLimitDegrees=45
  twistMotion=Free
 }
 {
  bone1Name=base_skirt_bk_l_01
  bone2Name=root
  swing1LimitDegrees=45
  swing1Motion=Free
  swing2LimitDegrees=45
  swing2Motion=Free
  twistLimitDegrees=45
  twistMotion=Free
 }
 {
  bone1Name=base_skirt_r_02
  bone2Name=root
  swing1LimitDegrees=45
  swing1Motion=Free
  swing2LimitDegrees=45
  swing2Motion=Free
  twistLimitDegrees=45
  twistMotion=Free
 }
 {bone1Name=dyn_skirt_bk_05, bone2Name=dyn_skirt_bk_04, swing1LimitDegrees=5, swing2LimitDegrees=25, twistLimitDegrees=0}
 {
  bone1Name=base_skirt_bk_01
  bone2Name=root
  swing1LimitDegrees=45
  swing1Motion=Free
  swing2LimitDegrees=45
  swing2Motion=Free
  twistLimitDegrees=45
  twistMotion=Free
 }
 {bone1Name=dyn_skirt_bk_01, bone2Name=base_skirt_bk_01, swing1LimitDegrees=5, swing2LimitDegrees=25, twistLimitDegrees=0}
 {bone1Name=dyn_skirt_bk_02, bone2Name=dyn_skirt_bk_01, swing1LimitDegrees=5, swing2LimitDegrees=15, twistLimitDegrees=0}
 {bone1Name=dyn_skirt_bk_03, bone2Name=dyn_skirt_bk_02, swing1LimitDegrees=5, swing2LimitDegrees=25, twistLimitDegrees=0}
 {bone1Name=dyn_skirt_bk_04, bone2Name=dyn_skirt_bk_03, swing1LimitDegrees=5, swing2LimitDegrees=25, twistLimitDegrees=0}
]
