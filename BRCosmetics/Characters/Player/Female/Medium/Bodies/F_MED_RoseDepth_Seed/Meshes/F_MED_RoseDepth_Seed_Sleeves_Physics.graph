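# Physics asset for the sleeves mesh of the RoseDepth Seed medium female player body, setting up physics for the lower left and right sleeve bones.
# Each side pairs a tiny capsule on the base bone with a larger, offset capsule on the dyn bone.
# One constraint per side joins the pair, with a wide first swing limit, a narrow second swing limit and a zero twist limit.
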PhysicsAsset /BRCosmetics/Characters/Player/Female/Medium/Bodies/F_MED_RoseDepth_Seed/Meshes/F_MED_RoseDepth_Seed_Sleeves_Physics
bodies {
 base_sleeve_down_l [{shapeType=Capsule, length=0.505, radius=0.505, shapeName=base_sleeve_down_l_capsule, rotation=<rot 90 0 0>}]
 base_sleeve_down_r [{shapeType=Capsule, length=0.505, radius=0.505, shapeName=base_sleeve_down_r_capsule, rotation=<rot 90 0 0>}]
 dyn_sleeve_down_l [{shapeType=Capsule, length=4.6592, radius=1.9513, shapeName=dyn_sleeve_down_l_capsule, center=<2 0 0>, rotation=<rot 90 0 0>}]
 dyn_sleeve_down_r [{shapeType=Capsule, length=4.6592, radius=1.9513, shapeName=dyn_sleeve_down_l_capsule, center=<2 0 0>, rotation=<rot 90 0 0>}]
}
constraints [
 {bone1Name=dyn_sleeve_down_l, bone2Name=base_sleeve_down_l, swing1LimitDegrees=50, swing2LimitDegrees=10, twistLimitDegrees=0}
 {bone1Name=dyn_sleeve_down_r, bone2Name=base_sleeve_down_r, swing1LimitDegrees=50, swing2LimitDegrees=10, twistLimitDegrees=0}
]
