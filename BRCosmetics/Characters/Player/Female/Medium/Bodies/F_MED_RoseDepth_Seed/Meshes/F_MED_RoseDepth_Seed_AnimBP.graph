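# Post-process-style AnimBP that animates the skirt, sleeves and dangling ropes of the RoseDepth Seed female cosmetic body.
# It copies the parent mesh pose through a Control Rig, then chains cached stages of additive bone offsets, RigidBody and AnimDynamics simulation, and corrective offsets.
# An extra RigidBody pass blends in on grind rails; the sleeve RigidBody node is unconnected.
# EventGraph derives alphas, gravity and simulation space from player state.
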AnimBlueprint /BRCosmetics/Characters/Player/Female/Medium/Bodies/F_MED_RoseDepth_Seed/Meshes/F_MED_RoseDepth_Seed_AnimBP
target_skeleton=&/F_MED_RoseDepth_Seed_Skeleton
graphs [
 AnimGraph {nodes [
   RigidBody!(ComponentPose, Alpha, OverrideWorldGravity:Vector="0,0,-980") -> (Pose) {
    bOverrideWorldGravity=true
    overridePhysicsAsset=&/F_MED_RoseDepth_Seed_Sleeves_Physics
    simulationSpace=BaseBoneSpace
    alphaBoolBlend {blendOption=Linear}
    alphaScaleBiasClamp {bInterpResult=true, interpSpeedDecreasing=5, interpSpeedIncreasing=5}
    baseBoneRef {boneName=spine_03}
    componentAppliedLinearAccClamp=<3000 3000 3000>
    componentLinearAccScale=<0.3 0.4 0.3>
    componentLinearVelScale=<0.3 0.4 0.3>
    overrideWorldGravity {z=-980}
    simSpaceSettings {
     maxAngularAcceleration=1000
     maxAngularVelocity=3
     maxLinearAcceleration=700
     maxLinearVelocity=500
     velocityScaleZ=0
     worldAlpha=0.3
     externalLinearDragV=<0.2 0.2>
    }
   }
   n1 = RigidBody(ComponentPose=Local To Component(LocalPose=Use cached pose 'CRCache_SkirtAdj'()), OverrideWorldGravity="0,0,-980") {
    bOverrideWorldGravity=true
    overridePhysicsAsset=&/F_MED_RoseDepth_Seed_Dyn_GrindRail
    simulationSpace=BaseBoneSpace
    alphaBoolBlend {blendOption=Linear}
    alphaScaleBiasClamp {bInterpResult=true, interpSpeedDecreasing=5, interpSpeedIncreasing=5}
    baseBoneRef {boneName=pelvis}
    componentAppliedLinearAccClamp=<3000 3000 3000>
    componentLinearAccScale=<0.3 0.3 0.3>
    componentLinearVelScale=<0.3 0.3 0.3>
    overrideWorldGravity {z=-980}
    simSpaceSettings {
     maxAngularAcceleration=1000
     maxAngularVelocity=3
     maxLinearAcceleration=700
     maxLinearVelocity=500
     velocityScaleZ=0
     worldAlpha=0.5
     externalLinearDragV=<0.2 0.2 0.2>
    }
   }
   Output Pose(Result=Blend Poses by bool(BlendPose_0=Component To Local(ComponentPose=@n1), BlendPose_1=Use cached pose 'CRCache_SkirtAdj'(), bActiveValue=Get bIsOnGrindRail()) {blendType=Linear}) {layerGroup=None}
   n2 = "Transform (Modify) Bone - Bone: dyn_skirt_r_01"(ComponentPose=Local To Component(LocalPose=Use cached pose 'PreSim Cache'()), Alpha=1, Translation="0,0,0", Rotation="0,0,0", Scale="1,1,1") {
    rotationMode=BMM_Additive
    rotationSpace=BCS_BoneSpace
    translationMode=BMM_Additive
    translationSpace=BCS_BoneSpace
    alphaBoolBlend {blendOption=Linear}
    boneToModify {boneName=dyn_skirt_r_01}
   }
   n3 = "Transform (Modify) Bone - Bone: dyn_skirt_fr_l_01"(ComponentPose=@n2, Alpha=1, Translation="0,0,3", Rotation="10,0,0", Scale="1,1,1") {
    rotationMode=BMM_Additive
    rotationSpace=BCS_BoneSpace
    translationMode=BMM_Additive
    translationSpace=BCS_BoneSpace
    alphaBoolBlend {blendOption=Linear}
    alphaScaleBiasClamp {bInterpResult=true, interpSpeedDecreasing=4, interpSpeedIncreasing=4}
    boneToModify {boneName=dyn_skirt_fr_l_01}
    rotation {pitch=10}
    translation {z=3}
   }
   n4 = "Transform (Modify) Bone - Bone: dyn_skirt_l_01"(ComponentPose=@n3, Alpha=1, Translation="0,0,3", Rotation="10,0,0", Scale="1,1,1") {
    rotationMode=BMM_Additive
    rotationSpace=BCS_BoneSpace
    translationMode=BMM_Additive
    translationSpace=BCS_BoneSpace
    alphaBoolBlend {blendOption=Linear}
    alphaScaleBiasClamp {bInterpResult=true, interpSpeedDecreasing=4, interpSpeedIncreasing=4}
    boneToModify {boneName=dyn_skirt_l_01}
    rotation {pitch=10}
    translation {z=3}
   }
   n5 = "Transform (Modify) Bone - Bone: dyn_skirt_bk_l_01"(ComponentPose=@n4, Alpha=1, Translation="0,0,3", Rotation="10,0,0", Scale="1,1,1") {
    rotationMode=BMM_Additive
    rotationSpace=BCS_BoneSpace
    translationMode=BMM_Additive
    translationSpace=BCS_BoneSpace
    alphaBoolBlend {blendOption=Linear}
    alphaScaleBiasClamp {bInterpResult=true, interpSpeedDecreasing=4, interpSpeedIncreasing=4}
    boneToModify {boneName=dyn_skirt_bk_l_01}
    rotation {pitch=10}
    translation {z=3}
   }
   n6 = "Transform (Modify) Bone - Bone: dyn_skirt_bk_01"(ComponentPose=@n5, Alpha=1, Translation="0,0,3", Rotation="10,0,0", Scale="1,1,1") {
    rotationMode=BMM_Additive
    rotationSpace=BCS_BoneSpace
    translationMode=BMM_Additive
    translationSpace=BCS_BoneSpace
    alphaBoolBlend {blendOption=Linear}
    alphaScaleBiasClamp {bInterpResult=true, interpSpeedDecreasing=4, interpSpeedIncreasing=4}
    boneToModify {boneName=dyn_skirt_bk_01}
    rotation {pitch=10}
    translation {z=3}
   }
   n7 = "Transform (Modify) Bone - Bone: dyn_skirt_r_01"(ComponentPose=@n6, Alpha=1, Translation="0,3,0", Rotation="0,10,0", Scale="1,1,1") {
    rotationMode=BMM_Additive
    rotationSpace=BCS_BoneSpace
    translationMode=BMM_Additive
    translationSpace=BCS_BoneSpace
    alphaBoolBlend {blendOption=Linear}
    alphaScaleBiasClamp {bInterpResult=true, interpSpeedDecreasing=4, interpSpeedIncreasing=4}
    boneToModify {boneName=dyn_skirt_r_01}
    rotation {yaw=10}
    translation {y=3}
   }
   n8 = "Transform (Modify) Bone - Bone: dyn_skirt_fr_l_01"!(ComponentPose=@n7, Alpha=1, Translation="0,0,0", Rotation="2,0,0") -> (Pose) {
    rotationMode=BMM_Additive
    rotationSpace=BCS_BoneSpace
    translationMode=BMM_Additive
    translationSpace=BCS_BoneSpace
    alphaBoolBlend {blendInTime=0.1, blendOutTime=0.1}
    alphaScaleBiasClamp {bInterpResult=true, interpSpeedDecreasing=4}
    boneToModify {boneName=dyn_skirt_fr_l_01}
    rotation {pitch=2}
   }
   n9 = "Transform (Modify) Bone - Bone: dyn_skirt_bk_01"!(ComponentPose=@n8, Alpha=1, Translation="0,0,0", Rotation="2,0,0") -> (Pose) {
    rotationMode=BMM_Additive
    rotationSpace=BCS_BoneSpace
    translationMode=BMM_Additive
    translationSpace=BCS_BoneSpace
    alphaBoolBlend {blendInTime=0.1, blendOutTime=0.1}
    alphaScaleBiasClamp {bInterpResult=true, interpSpeedDecreasing=4}
    boneToModify {boneName=dyn_skirt_bk_01}
    rotation {pitch=2}
   }
   n10 = "Transform (Modify) Bone - Bone: dyn_skirt_r_01"!(ComponentPose=@n9, Alpha=1, Translation="0,0,0", Rotation="2,-2,0") -> (Pose) {
    rotationMode=BMM_Additive
    rotationSpace=BCS_BoneSpace
    translationMode=BMM_Additive
    translationSpace=BCS_BoneSpace
    alphaBoolBlend {blendInTime=0.1, blendOutTime=0.1}
    alphaScaleBiasClamp {bInterpResult=true, interpSpeedDecreasing=4}
    boneToModify {boneName=dyn_skirt_r_01}
    rotation {pitch=2, yaw=-2}
   }
   Save cached pose 'CRCache_Input'(Pose=Component To Local(ComponentPose=@n10)) {cacheName=CRCache_Input}
   n11 = RigidBody(ComponentPose=Local To Component(LocalPose=Use cached pose 'CRCache_Input'()), OverrideWorldGravity="0,0,-980") {
    bOverrideWorldGravity=true
    overridePhysicsAsset=&/F_MED_RoseDepth_Seed_Dyn_skirt
    simulationSpace=BaseBoneSpace
    alphaBoolBlend {blendOption=Linear}
    alphaScaleBiasClamp {bInterpResult=true, interpSpeedDecreasing=5, interpSpeedIncreasing=5}
    baseBoneRef {boneName=pelvis}
    componentAppliedLinearAccClamp=<3000 3000 3000>
    componentLinearAccScale=<0.5 0.3 0.3>
    componentLinearVelScale=<0.5 0.3 0.3>
    overrideWorldGravity {z=-980}
    simSpaceSettings {
     maxAngularAcceleration=1000
     maxAngularVelocity=3
     maxLinearAcceleration=700
     maxLinearVelocity=500
     velocityScaleZ=0.3
     worldAlpha=0.6
     externalLinearDragV=<0.2 0.2>
    }
   }
   n12 = Anim Dynamics - Bone: base_sleeve_down_l!(ComponentPose=@n11, Alpha=1, GravityOverride:Vector="0,0,-980") -> (Pose) {
    angularDampingOverride=0.99
    angularSpringConstant=750
    bAngularSpring=true
    bOverrideAngularDamping=true
    bUseGravityOverride=true
    simulationSpace=RootRelative
    alphaBoolBlend {blendOption=Linear}
    boundBone {boneName=base_sleeve_down_l}
    componentAppliedLinearAccClamp=<3000 3000 3000>
    componentLinearAccScale=<0.4 0.4 0.4>
    componentLinearVelScale=<0.6 0.6 0.6>
    gravityOverride {z=-980}
    physicsBodyDefinitions [{
      boundBone {boneName=base_sleeve_down_l}
      boxExtents=<12 12 10>
      constraintSetup {angularLimitsMax=<0 5 8>, angularLimitsMin=<0 -5 -8>, angularTarget=<1 0 0>}
      localJointOffset=<1 0 0>
     }]
   }
   n13 = Anim Dynamics - Bone: dyn_sleeve_down_r!(ComponentPose=@n12, Alpha=1, GravityOverride:Vector="0,0,-980") -> (Pose) {
    angularDampingOverride=0.99
    angularSpringConstant=750
    bAngularSpring=true
    bOverrideAngularDamping=true
    bUseGravityOverride=true
    simulationSpace=RootRelative
    alphaBoolBlend {blendOption=Linear}
    boundBone {boneName=dyn_sleeve_down_r}
    componentAppliedLinearAccClamp=<3000 3000 3000>
    componentLinearAccScale=<0.4 0.4 0.4>
    componentLinearVelScale=<0.6 0.6 0.6>
    gravityOverride {z=-980}
    physicsBodyDefinitions [{
      boundBone {boneName=dyn_sleeve_down_r}
      boxExtents=<12 12 10>
      constraintSetup {angularLimitsMax=<0 5 8>, angularLimitsMin=<0 -5 -8>, angularTarget=<1 0 0>}
      localJointOffset=<1 0 0>
     }]
   }
   n14 = Anim Dynamics - Bone: dyn_Dangle_Rope_a_l!(SimulationSpace:AnimPhysSimSpaceType Enum=Component, ComponentPose=@n13, Alpha=0.7, ExternalForce="0,0,0") -> (Pose) {
    angularDampingOverride=1
    angularSpringConstant=20000
    bAngularSpring=true
    bOverrideAngularDamping=true
    bUseGravityOverride=true
    gravityScale=0.35
    alphaBoolBlend {blendOption=Linear}
    boundBone {boneName=dyn_Dangle_Rope_a_l}
    componentAppliedLinearAccClamp=<3000 3000 3000>
    componentLinearAccScale=<0.3 0.3 0.3>
    componentLinearVelScale=<0.3 0.3 0.3>
    gravityOverride {z=-980}
    physicsBodyDefinitions [{
      boundBone {boneName=dyn_Dangle_Rope_a_l}
      boxExtents=<25 35 35>
      constraintSetup {angularLimitsMax=<15 15 15>, angularLimitsMin=<-15 -15 -15>, angularTarget=<1 0 0>}
      localJointOffset=<3 0 0>
     }]
   }
   n15 = Anim Dynamics - Bone: dyn_Dangle_Rope_b_l!(SimulationSpace:AnimPhysSimSpaceType Enum=Component, ComponentPose=@n14, Alpha=0.7, ExternalForce="0,0,0") -> (Pose) {
    angularDampingOverride=1
    angularSpringConstant=20000
    bAngularSpring=true
    bOverrideAngularDamping=true
    bUseGravityOverride=true
    gravityScale=0.35
    alphaBoolBlend {blendOption=Linear}
    boundBone {boneName=dyn_Dangle_Rope_b_l}
    componentAppliedLinearAccClamp=<3000 3000 3000>
    componentLinearAccScale=<0.3 0.3 0.3>
    componentLinearVelScale=<0.3 0.3 0.3>
    gravityOverride {z=-980}
    physicsBodyDefinitions [{
      boundBone {boneName=dyn_Dangle_Rope_b_l}
      boxExtents=<25 35 35>
      constraintSetup {angularLimitsMax=<15 15 15>, angularLimitsMin=<-15 -15 -15>, angularTarget=<1 0 0>}
      localJointOffset=<3 0 0>
     }]
   }
   n16 = Anim Dynamics - Bone: dyn_Dangle_Rope_c_l!(SimulationSpace:AnimPhysSimSpaceType Enum=Component, ComponentPose=@n15, Alpha=0.7, ExternalForce="0,0,0") -> (Pose) {
    angularDampingOverride=1
    angularSpringConstant=20000
    bAngularSpring=true
    bOverrideAngularDamping=true
    bUseGravityOverride=true
    gravityScale=0.35
    alphaBoolBlend {blendOption=Linear}
    boundBone {boneName=dyn_Dangle_Rope_c_l}
    componentAppliedLinearAccClamp=<3000 3000 3000>
    componentLinearAccScale=<0.3 0.3 0.3>
    componentLinearVelScale=<0.3 0.3 0.3>
    gravityOverride {z=-980}
    physicsBodyDefinitions [{
      boundBone {boneName=dyn_Dangle_Rope_c_l}
      boxExtents=<25 35 35>
      constraintSetup {angularLimitsMax=<15 15 15>, angularLimitsMin=<-15 -15 -15>, angularTarget=<1 0 0>}
      localJointOffset=<3 0 0>
     }]
   }
   n17 = Anim Dynamics - Bone: dyn_Dangle_Rope_d_l!(SimulationSpace:AnimPhysSimSpaceType Enum=Component, ComponentPose=@n16, Alpha=0.7, ExternalForce="0,0,0") -> (Pose) {
    angularDampingOverride=1
    angularSpringConstant=20000
    bAngularSpring=true
    bOverrideAngularDamping=true
    bUseGravityOverride=true
    gravityScale=0.35
    alphaBoolBlend {blendOption=Linear}
    boundBone {boneName=dyn_Dangle_Rope_d_l}
    componentAppliedLinearAccClamp=<3000 3000 3000>
    componentLinearAccScale=<0.3 0.3 0.3>
    componentLinearVelScale=<0.3 0.3 0.3>
    gravityOverride {z=-980}
    physicsBodyDefinitions [{
      boundBone {boneName=dyn_Dangle_Rope_d_l}
      boxExtents=<25 35 35>
      constraintSetup {angularLimitsMax=<15 15 15>, angularLimitsMin=<-15 -15 -15>, angularTarget=<1 0 0>}
      localJointOffset=<3 0 0>
     }]
   }
   n18 = Anim Dynamics - Bone: dyn_Dangle_Rope_a_r!(SimulationSpace:AnimPhysSimSpaceType Enum=Component, ComponentPose=@n17, Alpha=0.7, ExternalForce="0,0,0") -> (Pose) {
    angularDampingOverride=1
    angularSpringConstant=20000
    bAngularSpring=true
    bOverrideAngularDamping=true
    bUseGravityOverride=true
    gravityScale=0.35
    alphaBoolBlend {blendOption=Linear}
    boundBone {boneName=dyn_Dangle_Rope_a_r}
    componentAppliedLinearAccClamp=<3000 3000>
    componentLinearAccScale=<0.3 0.3 0.3>
    componentLinearVelScale=<0.3 0.3 0.3>
    gravityOverride {z=-980}
    physicsBodyDefinitions [{
      boundBone {boneName=dyn_Dangle_Rope_a_r}
      boxExtents=<25 35 35>
      constraintSetup {angularLimitsMax=<15 15 15>, angularLimitsMin=<-15 -15 -15>, angularTarget=<1 0 0>}
      localJointOffset=<3 0 0>
     }]
   }
   n19 = Anim Dynamics - Bone: dyn_Dangle_Rope_b_r!(SimulationSpace:AnimPhysSimSpaceType Enum=Component, ComponentPose=@n18, Alpha=0.7, ExternalForce="0,0,0") -> (Pose) {
    angularDampingOverride=1
    angularSpringConstant=20000
    bAngularSpring=true
    bOverrideAngularDamping=true
    bUseGravityOverride=true
    gravityScale=0.35
    alphaBoolBlend {blendOption=Linear}
    boundBone {boneName=dyn_Dangle_Rope_b_r}
    componentAppliedLinearAccClamp=<3000 3000>
    componentLinearAccScale=<0.3 0.3 0.3>
    componentLinearVelScale=<0.3 0.3 0.3>
    gravityOverride {z=-980}
    physicsBodyDefinitions [{
      boundBone {boneName=dyn_Dangle_Rope_b_r}
      boxExtents=<25 35 35>
      constraintSetup {angularLimitsMax=<15 15 15>, angularLimitsMin=<-15 -15 -15>, angularTarget=<1 0 0>}
      localJointOffset=<3 0 0>
     }]
   }
   n20 = Anim Dynamics - Bone: dyn_Dangle_Rope_c_r!(SimulationSpace:AnimPhysSimSpaceType Enum=Component, ComponentPose=@n19, Alpha=0.7, ExternalForce="0,0,0") -> (Pose) {
    angularDampingOverride=1
    angularSpringConstant=20000
    bAngularSpring=true
    bOverrideAngularDamping=true
    bUseGravityOverride=true
    gravityScale=0.35
    alphaBoolBlend {blendOption=Linear}
    boundBone {boneName=dyn_Dangle_Rope_c_r}
    componentAppliedLinearAccClamp=<3000 3000>
    componentLinearAccScale=<0.3 0.3 0.3>
    componentLinearVelScale=<0.3 0.3 0.3>
    gravityOverride {z=-980}
    physicsBodyDefinitions [{
      boundBone {boneName=dyn_Dangle_Rope_c_r}
      boxExtents=<25 35 35>
      constraintSetup {angularLimitsMax=<15 15 15>, angularLimitsMin=<-15 -15 -15>, angularTarget=<1 0 0>}
      localJointOffset=<3 0 0>
     }]
   }
   n21 = Anim Dynamics - Bone: dyn_Dangle_Rope_d_r!(SimulationSpace:AnimPhysSimSpaceType Enum=Component, ComponentPose=@n20, Alpha=0.7, ExternalForce="0,0,0") -> (Pose) {
    angularDampingOverride=1
    angularSpringConstant=20000
    bAngularSpring=true
    bOverrideAngularDamping=true
    bUseGravityOverride=true
    gravityScale=0.35
    alphaBoolBlend {blendOption=Linear}
    boundBone {boneName=dyn_Dangle_Rope_d_r}
    componentAppliedLinearAccClamp=<3000 3000>
    componentLinearAccScale=<0.3 0.3 0.3>
    componentLinearVelScale=<0.3 0.3 0.3>
    gravityOverride {z=-980}
    physicsBodyDefinitions [{
      boundBone {boneName=dyn_Dangle_Rope_d_r}
      boxExtents=<25 35 35>
      constraintSetup {angularLimitsMax=<15 15 15>, angularLimitsMin=<-15 -15 -15>, angularTarget=<1 0 0>}
      localJointOffset=<3 0 0>
     }]
   }
   n22 = Anim Dynamics - Bone: dyn_sleeve_up_l!(SimulationSpace:AnimPhysSimSpaceType Enum=Component, ComponentPose=@n21, Alpha=0, ExternalForce="0,0,0") -> (Pose) {
    angularDampingOverride=1
    angularSpringConstant=20000
    bAngularSpring=true
    bOverrideAngularDamping=true
    bUseGravityOverride=true
    gravityScale=0.35
    alphaBoolBlend {blendOption=Linear}
    boundBone {boneName=dyn_sleeve_up_l}
    componentAppliedLinearAccClamp=<3000 3000 3000>
    componentLinearAccScale=<0.3 0.3 0.3>
    componentLinearVelScale=<1 1 1>
    gravityOverride {z=-980}
    physicsBodyDefinitions [{
      boundBone {boneName=dyn_sleeve_up_l}
      boxExtents=<30 50 50>
      constraintSetup {angularLimitsMax=<0 0 0>, angularLimitsMin=<0 -50 0>, angularTarget=<1 0 0>}
      localJointOffset=<4 0 0>
     }]
   }
   n23 = Anim Dynamics - Bone: dyn_sleeve_front_l!(SimulationSpace:AnimPhysSimSpaceType Enum=Component, ComponentPose=@n22, Alpha=0, ExternalForce="0,0,0") -> (Pose) {
    angularDampingOverride=1
    angularSpringConstant=20000
    bAngularSpring=true
    bOverrideAngularDamping=true
    bUseGravityOverride=true
    gravityScale=0.35
    alphaBoolBlend {blendOption=Linear}
    boundBone {boneName=dyn_sleeve_front_l}
    componentAppliedLinearAccClamp=<3000 3000 3000>
    componentLinearAccScale=<0.3 0.3 0.3>
    componentLinearVelScale=<1 1 1>
    gravityOverride {z=-980}
    physicsBodyDefinitions [{
      boundBone {boneName=dyn_sleeve_front_l}
      boxExtents=<30 50 50>
      constraintSetup {angularLimitsMax=<0 0 0>, angularLimitsMin=<-10 0 0>, angularTarget=<1 0 0>}
      localJointOffset=<4 0 0>
     }]
   }
   n24 = Anim Dynamics - Bone: dyn_sleeve_back_l!(SimulationSpace:AnimPhysSimSpaceType Enum=Component, ComponentPose=@n23, Alpha=0.5, ExternalForce="0,0,0") -> (Pose) {
    angularDampingOverride=1
    angularSpringConstant=20000
    bAngularSpring=true
    bOverrideAngularDamping=true
    bUseGravityOverride=true
    gravityScale=0.35
    alphaBoolBlend {blendOption=Linear}
    boundBone {boneName=dyn_sleeve_back_l}
    componentAppliedLinearAccClamp=<3000 3000 3000>
    componentLinearAccScale=<0.3 0.3 0.3>
    componentLinearVelScale=<1 1 1>
    gravityOverride {z=-980}
    physicsBodyDefinitions [{
      boundBone {boneName=dyn_sleeve_back_l}
      boxExtents=<30 50 50>
      constraintSetup {angularLimitsMax=<15 15 15>, angularLimitsMin=<-15 -15 -15>, angularTarget=<1 0 0>}
      localJointOffset=<4 0 0>
     }]
   }
   n25 = Anim Dynamics - Bone: dyn_sleeve_up_r!(SimulationSpace:AnimPhysSimSpaceType Enum=Component, ComponentPose=@n24, Alpha=0, ExternalForce="0,0,0") -> (Pose) {
    angularDampingOverride=1
    angularSpringConstant=20000
    bAngularSpring=true
    bOverrideAngularDamping=true
    bUseGravityOverride=true
    gravityScale=0.35
    alphaBoolBlend {blendOption=Linear}
    boundBone {boneName=dyn_sleeve_up_r}
    componentAppliedLinearAccClamp=<3000 3000 3000>
    componentLinearAccScale=<0.3 0.3 0.3>
    componentLinearVelScale=<1 1 1>
    gravityOverride {z=-980}
    physicsBodyDefinitions [{
      boundBone {boneName=dyn_sleeve_up_r}
      boxExtents=<30 50 50>
      constraintSetup {angularLimitsMax=<0 0 0>, angularLimitsMin=<0 -50 0>, angularTarget=<1 0 0>}
      localJointOffset=<-4 0 0>
     }]
   }
   n26 = Anim Dynamics - Bone: dyn_sleeve_front_r!(SimulationSpace:AnimPhysSimSpaceType Enum=Component, ComponentPose=@n25, Alpha=0, ExternalForce="0,0,0") -> (Pose) {
    angularDampingOverride=1
    angularSpringConstant=20000
    bAngularSpring=true
    bOverrideAngularDamping=true
    bUseGravityOverride=true
    gravityScale=0.35
    alphaBoolBlend {blendOption=Linear}
    boundBone {boneName=dyn_sleeve_front_r}
    componentAppliedLinearAccClamp=<3000 3000 3000>
    componentLinearAccScale=<0.3 0.3 0.3>
    componentLinearVelScale=<1 1 1>
    gravityOverride {z=-980}
    physicsBodyDefinitions [{
      boundBone {boneName=dyn_sleeve_front_r}
      boxExtents=<30 50 50>
      constraintSetup {angularLimitsMax=<10 0 0>, angularLimitsMin=<0 0 0>, angularTarget=<1 0 0>}
      localJointOffset=<-4 0 0>
     }]
   }
   n27 = Anim Dynamics - Bone: dyn_sleeve_back_r!(SimulationSpace:AnimPhysSimSpaceType Enum=Component, ComponentPose=@n26, Alpha=0.5, ExternalForce="0,0,0") -> (Pose) {
    angularDampingOverride=1
    angularSpringConstant=20000
    bAngularSpring=true
    bOverrideAngularDamping=true
    bUseGravityOverride=true
    gravityScale=0.35
    alphaBoolBlend {blendOption=Linear}
    boundBone {boneName=dyn_sleeve_back_r}
    componentAppliedLinearAccClamp=<3000 3000 3000>
    componentLinearAccScale=<0.3 0.3 0.3>
    componentLinearVelScale=<1 1 1>
    gravityOverride {z=-980}
    physicsBodyDefinitions [{
      boundBone {boneName=dyn_sleeve_back_r}
      boxExtents=<30 50 50>
      constraintSetup {angularLimitsMax=<15 15 15>, angularLimitsMin=<-15 -15 -15>, angularTarget=<1 0 0>}
      localJointOffset=<-4 0 0>
     }]
   }
   Save cached pose 'CRCache_Dyn'(Pose=Component To Local(ComponentPose=@n27)) {cacheName=CRCache_Dyn}
   n28 = "Transform (Modify) Bone - Bone: dyn_skirt_bk_l_01"!(ComponentPose=Local To Component(LocalPose=Use cached pose 'CRCache_Dyn'()), bAlphaBoolEnabled:bool=True, Translation="0,-2.5,0") -> (Pose) {
    alphaInputType=Bool
    translationMode=BMM_Additive
    alphaBoolBlend {blendInTime=0.1, blendOption=Linear, blendOutTime=0.1}
    boneToModify {boneName=dyn_skirt_bk_l_01}
    translation {y=-2.5}
   }
   n29 = "Transform (Modify) Bone - Bone: dyn_skirt_bk_01"!(ComponentPose=@n28, bAlphaBoolEnabled:bool=True, Translation="0,-2.5,0") -> (Pose) {
    alphaInputType=Bool
    translationMode=BMM_Additive
    alphaBoolBlend {blendInTime=0.1, blendOption=Linear, blendOutTime=0.1}
    boneToModify {boneName=dyn_skirt_bk_01}
    translation {y=-2.5}
   }
   n30 = "Transform (Modify) Bone - Bone: attach_backpack"!(ComponentPose=@n29, bAlphaBoolEnabled:bool=True, Translation="0,-4,0") -> (Pose) {
    alphaInputType=Bool
    translationMode=BMM_Additive
    translationSpace=BCS_BoneSpace
    alphaBoolBlend {blendInTime=0.1, blendOption=Linear, blendOutTime=0.1}
    boneToModify {boneName=attach_backpack}
    translation {y=-4}
   }
   n31 = "Transform (Modify) Bone - Bone: dyn_skirt_bk_l_01"!(ComponentPose=@n30, Alpha:float=1, Translation="0,0,4", Rotation:Rotator="0,0,0") -> (Pose) {
    rotationMode=BMM_Additive
    rotationSpace=BCS_BoneSpace
    translationMode=BMM_Additive
    translationSpace=BCS_BoneSpace
    alphaBoolBlend {blendInTime=0.1, blendOutTime=0.1}
    alphaScaleBiasClamp {bInterpResult=true, interpSpeedDecreasing=2.5, interpSpeedIncreasing=5}
    boneToModify {boneName=dyn_skirt_bk_l_01}
    translation {z=4}
   }
   n32 = "Transform (Modify) Bone - Bone: dyn_skirt_bk_01"!(ComponentPose=@n31, Alpha:float=1, Translation="0,0,6", Rotation:Rotator="0,0,0") -> (Pose) {
    rotationMode=BMM_Additive
    rotationSpace=BCS_BoneSpace
    translationMode=BMM_Additive
    translationSpace=BCS_BoneSpace
    alphaBoolBlend {blendInTime=0.1, blendOutTime=0.1}
    alphaScaleBiasClamp {bInterpResult=true, interpSpeedDecreasing=2.5, interpSpeedIncreasing=5}
    boneToModify {boneName=dyn_skirt_bk_01}
    translation {z=6}
   }
   n33 = "Transform (Modify) Bone - Bone: dyn_skirt_r_01"(ComponentPose=@n32, Alpha=1, Translation="0,-6,4", Rotation="0,0,0") {
    rotationMode=BMM_Additive
    rotationSpace=BCS_BoneSpace
    translationMode=BMM_Additive
    translationSpace=BCS_BoneSpace
    alphaBoolBlend {blendInTime=0.1, blendOutTime=0.1}
    alphaScaleBiasClamp {bInterpResult=true, interpSpeedDecreasing=2.5, interpSpeedIncreasing=5}
    boneToModify {boneName=dyn_skirt_r_01}
    translation {y=-6, z=4}
   }
   n34 = "Transform (Modify) Bone - Bone: base_skirt_fr_l_01"!(ComponentPose=@n33, Alpha=1, Translation="0,-5,4", Rotation="0,0,0") -> (Pose) {
    rotationMode=BMM_Additive
    rotationSpace=BCS_BoneSpace
    translationMode=BMM_Additive
    translationSpace=BCS_BoneSpace
    alphaBoolBlend {blendInTime=0.1, blendOutTime=0.1}
    alphaScaleBiasClamp {bInterpResult=true, interpSpeedDecreasing=2.5, interpSpeedIncreasing=5}
    boneToModify {boneName=base_skirt_fr_l_01}
    translation {y=-5, z=4}
   }
   n35 = "Transform (Modify) Bone - Bone: base_skirt_l_01"!(ComponentPose=@n34, Alpha=1, Translation="0,-5,4", Rotation="0,0,0") -> (Pose) {
    rotationMode=BMM_Additive
    rotationSpace=BCS_BoneSpace
    translationMode=BMM_Additive
    translationSpace=BCS_BoneSpace
    alphaBoolBlend {blendInTime=0.1, blendOutTime=0.1}
    alphaScaleBiasClamp {bInterpResult=true, interpSpeedDecreasing=2.5, interpSpeedIncreasing=5}
    boneToModify {boneName=base_skirt_l_01}
    translation {y=-5, z=4}
   }
   n36 = "Transform (Modify) Bone - Bone: base_Dangle_Rope_a_l"!(ComponentPose=@n35, bAlphaBoolEnabled:bool=True, Translation="0,0,0", Rotation="0,35,-70") -> (Pose) {
    alphaInputType=Bool
    rotationMode=BMM_Additive
    rotationSpace=BCS_BoneSpace
    translationMode=BMM_Additive
    translationSpace=BCS_BoneSpace
    alphaBoolBlend {blendInTime=0.1, blendOption=Linear, blendOutTime=0.1}
    boneToModify {boneName=base_Dangle_Rope_a_l}
    rotation {roll=-70, yaw=35}
   }
   n37 = "Transform (Modify) Bone - Bone: base_Dangle_Rope_c_l"!(ComponentPose=@n36, bAlphaBoolEnabled:bool=True, Translation="0,0,0", Rotation="-3,0,20") -> (Pose) {
    alphaInputType=Bool
    rotationMode=BMM_Additive
    rotationSpace=BCS_BoneSpace
    translationMode=BMM_Additive
    translationSpace=BCS_BoneSpace
    alphaBoolBlend {blendInTime=0.1, blendOption=Linear, blendOutTime=0.1}
    boneToModify {boneName=base_Dangle_Rope_c_l}
    rotation {pitch=-3, roll=20}
   }
   n38 = "Transform (Modify) Bone - Bone: base_Dangle_Rope_a_r"!(ComponentPose=@n37, bAlphaBoolEnabled:bool=True, Translation="0,0,1.5", Rotation="-68.5,0,-17") -> (Pose) {
    alphaInputType=Bool
    rotationMode=BMM_Additive
    rotationSpace=BCS_BoneSpace
    translationMode=BMM_Additive
    translationSpace=BCS_BoneSpace
    alphaBoolBlend {blendInTime=0.1, blendOption=Linear, blendOutTime=0.1}
    boneToModify {boneName=base_Dangle_Rope_a_r}
    rotation {pitch=-68.5, roll=-17}
    translation {z=1.5}
   }
   n39 = "Transform (Modify) Bone - Bone: base_Dangle_Rope_b_r"!(ComponentPose=@n38, bAlphaBoolEnabled:bool=True, Translation="-1.5,0,-2.5", Rotation="-20,0,0.144") -> (Pose) {
    alphaInputType=Bool
    rotationMode=BMM_Additive
    rotationSpace=BCS_BoneSpace
    translationMode=BMM_Additive
    translationSpace=BCS_BoneSpace
    alphaBoolBlend {blendInTime=0.1, blendOption=Linear, blendOutTime=0.1}
    boneToModify {boneName=base_Dangle_Rope_b_r}
    rotation {pitch=-20, roll=0.144}
    translation {x=-1.5, z=-2.5}
   }
   n40 = "Transform (Modify) Bone - Bone: base_Dangle_Rope_c_r"!(ComponentPose=@n39, bAlphaBoolEnabled:bool=True, Translation="-2,0,-2", Rotation="0,0,0") -> (Pose) {
    alphaInputType=Bool
    rotationMode=BMM_Additive
    rotationSpace=BCS_BoneSpace
    translationMode=BMM_Additive
    translationSpace=BCS_BoneSpace
    alphaBoolBlend {blendInTime=0.1, blendOption=Linear, blendOutTime=0.1}
    boneToModify {boneName=base_Dangle_Rope_c_r}
    translation {x=-2, z=-2}
   }
   n41 = "Transform (Modify) Bone - Bone: base_Dangle_Rope_d_r"!(ComponentPose=@n40, bAlphaBoolEnabled:bool=True, Translation="0,0,0", Rotation="-18,-10,18.5") -> (Pose) {
    alphaInputType=Bool
    rotationMode=BMM_Additive
    rotationSpace=BCS_BoneSpace
    translationMode=BMM_Additive
    translationSpace=BCS_BoneSpace
    alphaBoolBlend {blendInTime=0.1, blendOption=Linear, blendOutTime=0.1}
    boneToModify {boneName=base_Dangle_Rope_d_r}
    rotation=<rot -18 -10 18.5>
   }
   n42 = "Transform (Modify) Bone - Bone: dyn_skirt_fr_l_01"(ComponentPose=@n41, Alpha=1, Translation="0,0,1", Rotation="2,0,0") {
    rotationMode=BMM_Additive
    rotationSpace=BCS_BoneSpace
    translationMode=BMM_Additive
    translationSpace=BCS_BoneSpace
    alphaBoolBlend {blendInTime=0.1, blendOutTime=0.1}
    alphaScaleBiasClamp {bInterpResult=true, interpSpeedDecreasing=2.5, interpSpeedIncreasing=5}
    boneToModify {boneName=dyn_skirt_fr_l_01}
    rotation {pitch=2}
    translation {z=1}
   }
   n43 = "Transform (Modify) Bone - Bone: dyn_skirt_bk_l_01"(ComponentPose=@n42, Alpha=1, Translation="0,0,1", Rotation="0,0,40") {
    rotationMode=BMM_Additive
    rotationSpace=BCS_BoneSpace
    translationMode=BMM_Additive
    translationSpace=BCS_BoneSpace
    alphaBoolBlend {blendInTime=0.1, blendOutTime=0.1}
    alphaScaleBiasClamp {bInterpResult=true, interpSpeedDecreasing=2.5, interpSpeedIncreasing=5}
    boneToModify {boneName=dyn_skirt_bk_l_01}
    rotation {roll=40}
    translation {z=1}
   }
   n44 = "Transform (Modify) Bone - Bone: dyn_skirt_bk_01"(ComponentPose=@n43, Alpha=1, Translation="0,0,1", Rotation="30,-10,20") {
    rotationMode=BMM_Additive
    rotationSpace=BCS_BoneSpace
    translationMode=BMM_Additive
    translationSpace=BCS_BoneSpace
    alphaBoolBlend {blendInTime=0.1, blendOutTime=0.1}
    alphaScaleBiasClamp {bInterpResult=true, interpSpeedDecreasing=2.5, interpSpeedIncreasing=5}
    boneToModify {boneName=dyn_skirt_bk_01}
    rotation=<rot 30 -10 20>
    translation {z=1}
   }
   n45 = "Transform (Modify) Bone - Bone: dyn_skirt_r_01"(ComponentPose=@n44, Alpha=1, Translation="0,0,1", Rotation="40,-40,0") {
    rotationMode=BMM_Additive
    rotationSpace=BCS_BoneSpace
    translationMode=BMM_Additive
    translationSpace=BCS_BoneSpace
    alphaBoolBlend {blendInTime=0.1, blendOutTime=0.1}
    alphaScaleBiasClamp {bInterpResult=true, interpSpeedDecreasing=2.5, interpSpeedIncreasing=5}
    boneToModify {boneName=dyn_skirt_r_01}
    rotation {pitch=40, yaw=-40}
    translation {z=1}
   }
   n46 = "Transform (Modify) Bone - Bone: dyn_skirt_fr_l_01"(ComponentPose=@n45, Alpha=1, Translation="0,0,1", Rotation="-40,0,0") {
    rotationMode=BMM_Additive
    rotationSpace=BCS_BoneSpace
    translationMode=BMM_Additive
    translationSpace=BCS_BoneSpace
    alphaBoolBlend {blendInTime=0.1, blendOutTime=0.1}
    alphaScaleBiasClamp {bInterpResult=true, interpSpeedDecreasing=2.5, interpSpeedIncreasing=5}
    boneToModify {boneName=dyn_skirt_fr_l_01}
    rotation {pitch=-40}
    translation {z=1}
   }
   n47 = "Transform (Modify) Bone - Bone: dyn_skirt_l_01"!(ComponentPose=@n46, Alpha=1, Translation="0,0,1", Rotation="-30,-40,40") -> (Pose) {
    rotationMode=BMM_Additive
    rotationSpace=BCS_BoneSpace
    translationMode=BMM_Additive
    translationSpace=BCS_BoneSpace
    alphaBoolBlend {blendInTime=0.1, blendOutTime=0.1}
    alphaScaleBiasClamp {bInterpResult=true, interpSpeedDecreasing=2.5, interpSpeedIncreasing=5}
    boneToModify {boneName=dyn_skirt_l_01}
    rotation=<rot -30 -40 40>
    translation {z=1}
   }
   Save cached pose 'CRCache_SkirtAdj'(Pose=Component To Local(ComponentPose=@n47)) {cacheName=CRCache_SkirtAdj}
   n48 = Control Rig!(
    Source=Copy Pose From Mesh() {bUseAttachedParent=true}
    Alpha
    Movement Alpha:double=Get MovementAlpha() -> (MovementAlpha:double)
    is_DBNO_Carried:bool=Get bIsDBNOCarried()
    Is_DBNO:bool=Get bIsDBNO()
   ) -> (Pose) {
    alphaBoolBlend {blendOption=Linear}
    controlRigAssetReference {blueprintRigClass=&/F_MED_RoseDepth_Seed_CtrlRig.F_MED_RoseDepth_Seed_CtrlRig_C}
   }
   Save cached pose 'PreSim Cache'(Pose=@n48) {cacheName=PreSim Cache}
  ]}
 EventGraph {nodes [
   n1 = Event BlueprintUpdateAnimation()
   > n2 = Sequence()
   > n3 = Gravity Override State Function DEPRECATED(CharacterPartAnimInstance=Self-Reference(), GameTimeInSeconds=GetGameTimeInSeconds(), GravityOverrideParams=Get Gravity Override Params())
   > Set GravityOverride(GravityOverride=@n3)
   n4 = Get bIsMoving2D()
   > Set isMoving(isMoving:double=FInterpTo(Current=Get isMoving() -> (isMoving:double), Target="To Float (Boolean)"(InBool=@n4), DeltaTime=@n1.DeltaTimeX, InterpSpeed=5)) -> (Output_Get:double)
   n5 = Sequence!(execute=@n2.then_1) -> (then_0, then_1, then_2:Exec, then_3:Exec, then_4:Exec)
   > n6 = Branch(Condition=OR Boolean(A=Get bIsBaseballSliding(), B=Get bIsKneeSliding()))
   > Set SkirtDynAlpha(SkirtDynAlpha:double=0) -> (Output_Get:double)
   n7 = Get bIsMoving2D()
   n8 = Branch(execute=@n6.else, Condition=AND Boolean(A=Get bIsCrouching(), B=NOT Boolean(A=@n7)))
   > Set SkirtDynAlpha(SkirtDynAlpha=0.7)
   n9 = Branch(
    execute=@n8.else
    Condition=OR Boolean!(
     A=Get bIsSkydiveFloating()
     B=Get isZipline() -> (isZipline:bool)
     C:bool=Get bIsOnGrindRail()
     D:bool=Get bIsDBNO()
     E:bool=Get bIsTargeting()
    ) -> (ReturnValue)
   )
   > Set SkirtDynAlpha(SkirtDynAlpha=0.5)
   Set SkirtDynAlpha(execute=@n9.else, SkirtDynAlpha=0.7)
   n10 = Sequence(execute=@n5.then_4)
   > n11 = PlayerMovement(AnimInstance=Self-Reference(), XDotProductCoefficient=25, YDotProductCoefficient=25, ZDotProductCoefficient=25)
   > Set MovementAlpha(MovementAlpha:double="Absolute (Float)"(A=float * float(A=@n11.YMovement, B=0.045) {is=title K2Node_PromotableOperator})) -> (Output_Get:double)
   n12 = Get bIsSkyDiving()
   n13 = Get bIsFalling()
   n14 = Get bIsJumping()
   n15 = Branch(execute=@n10.then_1, Condition=AND Boolean(A=OR Boolean!(A=@n13, B=@n14) -> (ReturnValue), B=NOT Boolean(A=@n12)))
   > Set JumpingFalling_Skirt_Alpha(JumpingFalling_Skirt_Alpha:double=1) -> (Output_Get:double)
   Set JumpingFalling_Skirt_Alpha(execute=@n15.else, JumpingFalling_Skirt_Alpha=0)
   Set Is Playing Emote Alpha(execute=@n10.then_2, Is Playing Emote Alpha:double="To Float (Boolean)"(InBool=Get bIsPlayingEmote())) -> (Output_Get:double)
   Set Is Baseball Sliding Alpha(execute=@n10.then_3, Is Baseball Sliding Alpha:double="To Float (Boolean)"(InBool=Get bIsBaseballSliding())) -> (Output_Get:double)
   Set Is Playing Melee Anim Alpha(execute=@n10.then_4, Is Playing Melee Anim Alpha:double="To Float (Boolean)"(InBool=Get bIsPlayingMeleeAnim())) -> (Output_Get:double)
   n16 = Branch(execute=@n5.then_2, Condition=Get bIsInFrontEnd())
   > Set SimlationSpace(SimlationSpace:AnimPhysSimSpaceType Enum=World) -> (Output_Get:AnimPhysSimSpaceType Enum)
   Set SimlationSpace(execute=@n16.else, SimlationSpace=Component)
   Event OnParachuteOpened()
   > Switch on EGliderType(Selection=Get PlayerGliderType() -> (PlayerGliderType=HangGlider))
   > Set HangGliding(HangGliding:bool=true) -> (Output_Get:bool)
   Event OnParachuteClosed()
   > Set HangGliding(HangGliding=false)
   Event OnCharacterCustomizationLoaded()
   > ResetDynamics(InTeleportType=TeleportPhysics)
   > ResetDynamics(InTeleportType=ResetPhysics)
   n17 = Event BlueprintProcessAnimInputEvent()
   > n18 = Branch(Condition=IsNamed(self=@n17.AnimInputEvent, InEventName=aie.Player.Zipline.Entered))
   > Set isZipline(isZipline:bool=true) -> (Output_Get:bool)
   Branch(execute=@n18.else, Condition=IsNamed(self=@n17.AnimInputEvent, InEventName=aie.Player.Zipline.Exited))
   > Set isZipline(isZipline=false)
  ]}
]
variables [
 Gravity Override Params {type=GravityOverrideParamsStruct}
 GravityOverride {type=Vector}
 HangGliding {type=boolean}
 Is Baseball Sliding Alpha {type=number}
 Is Playing Emote Alpha {type=number}
 Is Playing Melee Anim Alpha {type=number}
 isMoving {type=number}
 isZipline {type=boolean}
 JumpingFalling_Skirt_Alpha {type=number}
 MovementAlpha {type=number}
 SimlationSpace {type=AnimPhysSimSpaceType}
 SkirtDynAlpha {type=number}
]
